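# Unlit additive flare material, named for alerts, that draws a camera-facing, radial sphere-mask glow in a fixed or time-of-day color, chosen by static switch.
# World position offset billboards the mesh and nudges it toward the camera.
# With depth testing disabled, one scene depth sample at the biased pivot, compared to pixel depth, fades the whole flare when occluded.
# Emissive is scaled by inverse eye adaptation, then attributes pass through MF_LockdownPulse.
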Material /VividRazorVaultGameplay/Proto/Materials/M_AlertFlare
graph {
 properties {
  bDisableDepthTest=true
  bUseMaterialAttributes=true
  blendMode=BLEND_Additive
  pixelDepthOffsetMode=PDOM_Legacy
  shadingModel=MSM_Unlit
  usedShadingModels=Unlit
 }
 nodes [
  n1 = Constant() {r=0.5}
  n2 = SphereMask(A=TextureCoordinate(), B=@n1, Radius=@n1, Hardness=ScalarParameter() {defaultValue=0.1, parameterName=FlareHardness})
  n3 = ScalarParameter() {defaultValue=4, parameterName=FlareExp}
  n4 = MaterialFunctionCall(
   Color_Morning=VectorParameter() {parameterName=FlareColorMorning, defaultValue=<rgba 1 0 0 0>}.RGB
   Color_Day=VectorParameter() {parameterName=FlareColorDay, defaultValue=<rgba 0 1 0 0>}.RGB
   Color_Evening=VectorParameter() {parameterName=FlareColorEvening, defaultValue=<rgba 0 0 1 0>}.RGB
   Color_Night=VectorParameter() {parameterName=FlareColorNight, defaultValue=<rgba 1 1 1 1>}.RGB
  ) {materialFunction=&Color_Time_of_Day}
  n5 = VectorParameter() {parameterName=FlareColor, defaultValue=<rgba 0.0681 0.151 0.4 0>}
  n6 = Add(A=WorldPosition().XYZ, B=Subtract(A=ObjectPositionWS(), B=WorldPosition().XYZ))
  n7 = CameraPositionWS()
  n8 = ObjectPositionWS()
  n9 = Multiply(A=Normalize(VectorInput=Subtract(A=@n7, B=@n8)), B=ScalarParameter() {defaultValue=25, parameterName=CameraBias})
  n10 = SceneDepth(UVs=MaterialFunctionCall(World Position=Add(A=@n6, B=@n9)) {materialFunction=&TransformToClipSpace}.Screen aligned 0-1 UV)
  n11 = Subtract(
   A=PixelDepth().R
   B=MaterialFunctionCall(Vector 3=@n9) {materialFunction=/Engine/Functions/Engine_MaterialFunctions02/Utility/VectorLength}.V3 Length
  )
  n12 = Multiply(
   A=Multiply(A=Power(Base=@n2, Exp=@n3) {constExponent=3}, B=StaticSwitchParameter(True=@n4.Result, False=@n5.RGB) {parameterName=UseTODFlareColors})
   B=Saturate(None=Subtract(A=@n10.R, B=@n11))
  )
  n13 = EyeAdaptationInverse()
  n14 = MaterialFunctionCall!(
   WorldPosition=WorldPosition().XYZ
   PivotLocation
   WS Normals
   Custom Object Basis 1=Constant3Vector() {constant=<rgba 0 1 0 1>}
   Custom Object Basis 2=Constant3Vector() {constant=<rgba 0 0 -1 0>}
   Custom Object Basis 3
  ) {materialFunction=/Engine/Functions/Engine_MaterialFunctions02/WorldPositionOffset/AlignMeshToTheCamera}
  MaterialFunctionCall!(MA In=SetMaterialAttributes!(
    MaterialAttributes
    Emissive Color=Multiply(A=@n12, B=@n13.EyeAdaptationInverse)
    World Position Offset=Add(A=@n9, B=@n14.World Position Offset)
   ) {attributeSetTypes [#1, #2]}) {materialFunction=/Game/Global/MaterialFunctions/MF_LockdownPulse}
 ]
}
