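# M_LF_AlertPulse is a prototype light function material, named for an alert pulse effect on lights.
# Its graph holds a single call to the shared MF_LockdownPulse material function, with no local parameters or math.
# The graph shows no link from that call to a material output, and any pulse logic lives inside the function.
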Material /VividRazorVaultGameplay/Proto/Materials/M_LF_AlertPulse
graph {
 properties {materialDomain=MD_LightFunction, pixelDepthOffsetMode=PDOM_Legacy, usedShadingModels=Default Lit}
 nodes [MaterialFunctionCall!(MA In) {materialFunction=/Game/Global/MaterialFunctions/MF_LockdownPulse}]
}
